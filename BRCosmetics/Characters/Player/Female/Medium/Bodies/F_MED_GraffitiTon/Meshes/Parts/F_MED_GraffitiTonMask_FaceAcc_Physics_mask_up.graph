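# Physics asset for the GraffitiTon female character's mask face accessory, apparently its mask-up state, providing collision for braid, mask and strap dynamics.
# Capsules cover the head, neck, upper spine, clavicles and upper arms, with large offset capsules on spine_05 and clavicle_l.
# Two five-link braid chains hang from MSTR bones through swing-limited, low-twist constraints.
# Mask and strap bones have capsules but no constraints.
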PhysicsAsset /BRCosmetics/Characters/Player/Female/Medium/Bodies/F_MED_GraffitiTon/Meshes/Parts/F_MED_GraffitiTonMask_FaceAcc_Physics_mask_up
bodies {
 clavicle_l [
  {shapeType=Capsule, length=10.7406, radius=3.8, shapeName=clavicle_l_capsule, center=<5.3 0 0>, rotation=<rot 90 0 0>}
  {
   shapeType=Capsule
   length=15
   radius=16.7857
   shapeName=clavicle_l_capsule
   center=<6.4092 -21.2456 6.1711>
   rotation=<rot -88.293 179.9966 -177.0124>
  }
 ]
 clavicle_r [{
   shapeType=Capsule
   length=6.5156
   radius=4.3
   shapeName=clavicle_r_capsule
   center=<-3.8185 1.2852 0.6005>
   rotation=<rot 85.6284 -4.4798 -0.002>
  }]
 dyn_braid_l_01 [{shapeType=Capsule, length=3, radius=1.5, shapeName=dyn_braid_l_01_capsule, center=<5.2328 -0.2227 0.5023>, rotation=<rot 0 91.7119 -88.222>}]
 dyn_braid_l_02 [{shapeType=Capsule, length=2, radius=1.6, shapeName=dyn_braid_l_02_capsule, center=<3.5 0 0>, rotation=<rot 88.6626 -0.0034 5.2609>}]
 dyn_braid_l_03 [{shapeType=Capsule, length=2.3, radius=1.5, shapeName=dyn_braid_l_03_capsule, center=<2.5 0 0>, rotation=<rot 90 119.3681 123.269>}]
 dyn_braid_l_04 [{shapeType=Capsule, length=2.6308, radius=1.5, shapeName=dyn_braid_l_04_capsule, center=<2.4235 0 0>, rotation=<rot 90 0 0>}]
 dyn_braid_l_05 [{shapeType=Capsule, length=5, radius=3, shapeName=dyn_braid_l_05_capsule, center=<3 -0.0543 0.0348>, rotation=<rot 0 -93.712 -87.1847>}]
 dyn_braid_l_MSTR [{shapeType=Capsule, length=0.505, radius=0.505, shapeName=dyn_braid_l_MSTR_capsule, rotation=<rot 0 -90 0>}]
 dyn_braid_r_01 [{shapeType=Capsule, length=3.8, radius=1.8, shapeName=dyn_braid_l_01_capsule, center=<5 -0.2227 0.5023>, rotation=<rot 0 91.7119 -88.222>}]
 dyn_braid_r_02 [{shapeType=Capsule, length=2.317, radius=2.5, shapeName=dyn_braid_l_02_capsule, center=<3.5 0 0>, rotation=<rot 88.6614 -0.0001 1.9763>}]
 dyn_braid_r_03 [{shapeType=Capsule, length=3.4862, radius=2.1377, shapeName=dyn_braid_l_03_capsule, center=<2.0569 0 0>, rotation=<rot 84.3066 0.0001 0.0002>}]
 dyn_braid_r_04 [{shapeType=Capsule, length=2.6308, radius=1.8948, shapeName=dyn_braid_l_04_capsule, center=<2.4235 0 0>, rotation=<rot 90 0 0>}]
 dyn_braid_r_05 [{shapeType=Capsule, length=5, radius=3, shapeName=dyn_braid_l_05_capsule, center=<3 -0.0543 0.0348>, rotation=<rot 0 -93.712 -87.1847>}]
 dyn_braid_r_MSTR [{shapeType=Capsule, length=0.505, radius=0.505, shapeName=dyn_braid_r_MSTR_capsule, rotation=<rot 0 -90 0>}]
 dyn_mask_l [{
   shapeType=Capsule
   length=4.7389
   radius=3.4549
   shapeName=dyn_mask_l_capsule
   center=<4.6912 -0.0686 -0.46>
   rotation=<rot 0 -95.1445 -87.9388>
  }]
 dyn_mask_off_l [{shapeType=Capsule, length=3, radius=3, shapeName=dyn_mask_off_l_capsule, center=<4.7437 -1 -2.4787>, rotation=<rot 90 -32.2645 -19.0026>}]
 dyn_mask_off_r [{shapeType=Capsule, length=7, radius=2.5, shapeName=dyn_mask_off_r_capsule, center=<2.1618 2.3325 -1.2252>, rotation=<rot 90 180 160.4542>}]
 dyn_mask_r [{
   shapeType=Capsule
   length=4.7344
   radius=3.4569
   shapeName=dyn_mask_r_capsule
   center=<4.6945 0.064 -0.4586>
   rotation=<rot 0 -84.7032 -89.0499>
  }]
 dyn_unmask_strap_l [{
   shapeType=Capsule
   length=1.4263
   radius=0.6
   shapeName=dyn_unmask_strap_l_capsule
   center=<0.0782 -1.3033 -0.094>
   rotation=<rot 0 -173.029 -89.4938>
  }]
 dyn_unmask_strap_r [{
   shapeType=Capsule
   length=2
   radius=1
   shapeName=dyn_unmask_strap_r_capsule
   center=<0.4394 -1.2303 -0.084>
   rotation=<rot 0.0429 153.6113 -89.8886>
  }]
 head [{shapeType=Capsule, length=9.9618, radius=4.8, shapeName=head_capsule, center=<2.5415 1.7516 0>, rotation=<rot 90 0 0>}]
 neck_02 [{shapeType=Capsule, length=13.6209, radius=4.3, shapeName=neck_01_capsule, center=<0 1.6059 0>, rotation=<rot 90 0 0>}]
 spine_04 [{shapeType=Capsule, length=8, radius=11, shapeName=spine_04_capsule, center=<-0.2202 0.1518 0>, rotation=<rot 0 -90 0>}]
 spine_05 [
  {shapeType=Capsule, length=37.1931, radius=30, shapeName=spine_05_capsule, center=<13.644 -36.2312 -0.2064>}
  {shapeType=Capsule, length=16.2161, radius=5.3, shapeName=spine_05_capsule, center=<1.3941 0.5699 0>}
 ]
 upperarm_l [{shapeType=Capsule, length=20, radius=4, shapeName=upperarm_l_capsule, center=<12.4662 0 0>, rotation=<rot 90 0 0>}]
 upperarm_r [{shapeType=Capsule, length=25.6391, radius=3.5, shapeName=upperarm_r_capsule, center=<-12.6256 1.5117 -0.9815>, rotation=<rot 90 0 0>}]
}
constraints [
 {bone1Name=dyn_braid_r_05, bone2Name=dyn_braid_r_04, swing1LimitDegrees=35, swing2LimitDegrees=35, twistLimitDegrees=2}
 {bone1Name=dyn_braid_r_01, bone2Name=dyn_braid_r_MSTR, swing1LimitDegrees=40, swing2LimitDegrees=40, twistLimitDegrees=2}
 {bone1Name=dyn_braid_r_02, bone2Name=dyn_braid_r_01, swing1LimitDegrees=40, swing2LimitDegrees=40, twistLimitDegrees=2}
 {bone1Name=dyn_braid_r_03, bone2Name=dyn_braid_r_02, swing1LimitDegrees=35, swing2LimitDegrees=50, twistLimitDegrees=2}
 {bone1Name=dyn_braid_r_04, bone2Name=dyn_braid_r_03, swing1LimitDegrees=35, swing2LimitDegrees=35, twistLimitDegrees=2}
 {bone1Name=dyn_braid_l_01, bone2Name=dyn_braid_l_MSTR, swing1LimitDegrees=35, swing2LimitDegrees=35, twistLimitDegrees=2}
 {bone1Name=dyn_braid_l_02, bone2Name=dyn_braid_l_01, swing1LimitDegrees=35, swing2LimitDegrees=35, twistLimitDegrees=2}
 {bone1Name=dyn_braid_l_03, bone2Name=dyn_braid_l_02, swing1LimitDegrees=50, swing2LimitDegrees=50, twistLimitDegrees=2}
 {bone1Name=dyn_braid_l_04, bone2Name=dyn_braid_l_03, swing1LimitDegrees=35, swing2LimitDegrees=35, twistLimitDegrees=2}
 {bone1Name=dyn_braid_l_05, bone2Name=dyn_braid_l_04, swing1LimitDegrees=35, swing2LimitDegrees=35, twistLimitDegrees=2}
]
