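# Post-process style face accessory AnimBP for GraffitiTon, animating the mask, braids and small hair.
# Parent pose feeds a Control Rig, mask-up or mask-down RigidBody picked by the Stage2 variant tag, then hair AnimDynamics.
# Curve-driven braid constraints and mask offsets for skydive floating or harvesting montages finish the output pose.
# Several braid constraint, look-at and AnimDynamics chains are disconnected from the output.
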AnimBlueprint /BRCosmetics/Characters/Player/Female/Medium/Bodies/F_MED_GraffitiTon/Meshes/Parts/F_MED_GraffitiTon_FaceAcc_AnimBP
target_skeleton=&/F_MED_GraffitiTon_FaceAcc_Skeleton
graphs [
 AnimGraph {nodes [
   n1 = Get mask turn off() -> (mask turn off:double)
   n2 = Get mask_turn_off2() -> (mask_turn_off2:double)
   n3 = "Constraint - dyn_braid_l_01 (PPPP)"!(
    ComponentPose
    Alpha=@n2
    ConstraintWeights_0=0.176
    ConstraintWeights_1:float=0.2007
    ConstraintWeights_2:float=0.1946
    ConstraintWeights_3:float=0.4287
   ) -> (Pose) {
    alphaBoolBlend {blendOption=Linear}
    alphaScaleBiasClamp {interpSpeedDecreasing=20}
    boneToModify {boneName=dyn_braid_l_01}
    constraintSetup [{targetBone {boneName=spine_05}}, {targetBone {boneName=neck_01}}, {targetBone {boneName=head}}, {targetBone {boneName=neck_02}}]
    constraintWeights [1,1,1,1]
   }
   n4 = "Constraint - dyn_braid_l_02 (PPP)"!(ComponentPose=@n3, Alpha=@n2, ConstraintWeights_0=0.718, ConstraintWeights_1:float=0.2777, ConstraintWeights_2:float=0.0043) -> (Pose) {
    alphaBoolBlend {blendOption=Linear}
    alphaScaleBiasClamp {interpSpeedDecreasing=20}
    boneToModify {boneName=dyn_braid_l_02}
    constraintSetup [{targetBone {boneName=spine_05}}, {targetBone {boneName=clavicle_l}}, {targetBone {boneName=neck_02}}]
    constraintWeights [1,1,1]
   }
   n5 = "Constraint - dyn_braid_l_03 (PPP)"!(ComponentPose=@n4, Alpha=@n2, ConstraintWeights_0=0.0207, ConstraintWeights_1:float=0.7578, ConstraintWeights_2:float=0.2215) -> (Pose) {
    alphaBoolBlend {blendOption=Linear}
    alphaScaleBiasClamp {interpSpeedDecreasing=20}
    boneToModify {boneName=dyn_braid_l_03}
    constraintSetup [{targetBone {boneName=spine_04}}, {targetBone {boneName=spine_05}}, {targetBone {boneName=clavicle_l}}]
    constraintWeights [1,1,1]
   }
   n6 = "Constraint - dyn_braid_r_01 (PPPP)"!(
    ComponentPose=@n5
    Alpha=@n2
    ConstraintWeights_0=0.1752
    ConstraintWeights_1:float=0.2008
    ConstraintWeights_2:float=0.1961
    ConstraintWeights_3:float=0.4279
   ) -> (Pose) {
    alphaBoolBlend {blendOption=Linear}
    boneToModify {boneName=dyn_braid_r_01}
    constraintSetup [{targetBone {boneName=spine_05}}, {targetBone {boneName=neck_01}}, {targetBone {boneName=head}}, {targetBone {boneName=neck_02}}]
    constraintWeights [1,1,1,1]
   }
   n7 = "Constraint - dyn_braid_r_02 (PPPP)"!(
    ComponentPose=@n6
    Alpha=@n2
    ConstraintWeights_0=0.003
    ConstraintWeights_1:float=0.7357
    ConstraintWeights_2:float=0.257
    ConstraintWeights_3:float=0.0043
   ) -> (Pose) {
    alphaBoolBlend {blendOption=Linear}
    boneToModify {boneName=dyn_braid_r_02}
    constraintSetup [{targetBone {boneName=spine_04}}, {targetBone {boneName=spine_05}}, {targetBone {boneName=clavicle_r}}, {targetBone {boneName=neck_02}}]
    constraintWeights [1,1,1,1]
   }
   n8 = "Constraint - dyn_braid_r_03 (PPP)"!(ComponentPose=@n7, Alpha=@n2, ConstraintWeights_0=0.0145, ConstraintWeights_1:float=0.7615, ConstraintWeights_2:float=0.224) -> (Pose) {
    alphaBoolBlend {blendOption=Linear}
    boneToModify {boneName=dyn_braid_r_03}
    constraintSetup [{targetBone {boneName=spine_04}}, {targetBone {boneName=spine_05}}, {targetBone {boneName=clavicle_r}}]
    constraintWeights [1,1,1]
   }
   n9 = "Transform (Modify) Bone - Bone: dyn_mask_off_mstr"(ComponentPose=@n8, Alpha=@n2, Translation="0,0,0", Rotation="0,0,0", Scale="0.97,0.97,0.97") {
    rotationMode=BMM_Additive
    rotationSpace=BCS_BoneSpace
    scaleMode=BMM_Replace
    scaleSpace=BCS_BoneSpace
    translationMode=BMM_Additive
    translationSpace=BCS_BoneSpace
    alphaBoolBlend {blendOption=Linear}
    boneToModify {boneName=dyn_mask_off_mstr}
    scale=<0.97 0.97 0.97>
   }
   n10 = Look At - Bone: dyn_braid_l_04(ComponentPose=@n9, Alpha=@n1) {
    alphaBoolBlend {blendOption=Linear}
    boneToModify {boneName=dyn_braid_l_04}
    lookAtLocation=<0 12 -8>
    lookAtTarget {boneReference {boneName=spine_05}}
    lookAt_Axis {axis=<1 0>}
   }
   n11 = Look At - Bone: dyn_braid_r_04(ComponentPose=@n10, Alpha=@n1) {
    alphaBoolBlend {blendOption=Linear}
    boneToModify {boneName=dyn_braid_r_04}
    lookAtLocation=<0 12 7>
    lookAtTarget {boneReference {boneName=spine_05}}
    lookAt_Axis {axis=<1 0>}
   }
   n12 = Anim Dynamics - Bone: dyn_braid_l_02(ComponentPose=@n11, Alpha=1, ExternalForce="0,0,0") {
    bUseGravityOverride=true
    bUseSphericalLimits=true
    alphaBoolBlend {blendOption=Linear}
    boundBone {boneName=dyn_braid_l_02}
    componentAppliedLinearAccClamp=<3000 3000 3000>
    componentLinearAccScale=<0.2 0.2 0.2>
    componentLinearVelScale=<0.2 0.2 0.2>
    gravityOverride {z=-500}
    physicsBodyDefinitions [{
      collisionType=CustomSphere
      sphereCollisionRadius=5
      boundBone {boneName=dyn_braid_l_02}
      boxExtents=<10 10 10>
      constraintSetup {angularLimitsMax=<0 7 8>, angularLimitsMin=<0 -7 0>, angularTarget=<0 0 0>}
      localJointOffset=<2 0 0>
     }]
    sphericalLimits [{limitRadius=12, drivingBone {boneName=spine_05}, sphereLocalOffset=<-10 5 0>}]
   }
   n13 = Anim Dynamics - Bone: dyn_braid_l_03(ComponentPose=@n12, Alpha=1, ExternalForce="0,0,0") {
    bUseGravityOverride=true
    bUseSphericalLimits=true
    alphaBoolBlend {blendOption=Linear}
    boundBone {boneName=dyn_braid_l_03}
    componentAppliedLinearAccClamp=<3000 3000 3000>
    componentLinearAccScale=<0.2 0.2 0.2>
    componentLinearVelScale=<0.2 0.2 0.2>
    gravityOverride {z=-500}
    physicsBodyDefinitions [{
      collisionType=CustomSphere
      sphereCollisionRadius=5
      boundBone {boneName=dyn_braid_l_03}
      boxExtents=<10 10 10>
      constraintSetup {angularLimitsMax=<0 7 10>, angularLimitsMin=<0 -7 0>, angularTarget=<0 0 0>}
      localJointOffset=<2 0 0>
     }]
    sphericalLimits [{limitRadius=12, drivingBone {boneName=spine_05}, sphereLocalOffset=<-10 5 0>}]
   }
   n14 = Anim Dynamics - Bone: dyn_braid_l_04(ComponentPose=@n13, Alpha=1, ExternalForce="0,0,0") {
    bUseGravityOverride=true
    bUseSphericalLimits=true
    alphaBoolBlend {blendOption=Linear}
    boundBone {boneName=dyn_braid_l_04}
    componentAppliedLinearAccClamp=<3000 3000 3000>
    componentLinearAccScale=<0.2 0.2 0.2>
    componentLinearVelScale=<0.2 0.2 0.2>
    gravityOverride {z=-500}
    physicsBodyDefinitions [{
      collisionType=CustomSphere
      sphereCollisionRadius=5
      boundBone {boneName=dyn_braid_l_04}
      boxExtents=<10 10 10>
      constraintSetup {angularLimitsMax=<0 7 15>, angularLimitsMin=<0 -7 0>, angularTarget=<0 0 0>}
      localJointOffset=<2 0 0>
     }]
    sphericalLimits [{limitRadius=12, drivingBone {boneName=spine_05}, sphereLocalOffset=<-10 9 -5>}]
   }
   n15 = Anim Dynamics - Bone: dyn_braid_r_02(ComponentPose=@n14, Alpha=1, ExternalForce="0,0,0") {
    bUseGravityOverride=true
    bUseSphericalLimits=true
    alphaBoolBlend {blendOption=Linear}
    boundBone {boneName=dyn_braid_r_02}
    componentAppliedLinearAccClamp=<3000 3000 3000>
    componentLinearAccScale=<0.2 0.2 0.2>
    componentLinearVelScale=<0.2 0.2 0.2>
    gravityOverride {z=-500}
    physicsBodyDefinitions [{
      collisionType=CustomSphere
      sphereCollisionRadius=5
      boundBone {boneName=dyn_braid_r_02}
      boxExtents=<10 10 10>
      constraintSetup {angularLimitsMax=<0 7 0>, angularLimitsMin=<0 -7 -8>, angularTarget=<0 0 0>}
      localJointOffset=<2 0 0>
     }]
    sphericalLimits [{limitRadius=12, drivingBone {boneName=spine_05}, sphereLocalOffset=<-10 5 0>}]
   }
   n16 = Anim Dynamics - Bone: dyn_braid_r_03(ComponentPose=@n15, Alpha=1, ExternalForce="0,0,0") {
    bUseGravityOverride=true
    bUseSphericalLimits=true
    alphaBoolBlend {blendOption=Linear}
    boundBone {boneName=dyn_braid_r_03}
    componentAppliedLinearAccClamp=<3000 3000 3000>
    componentLinearAccScale=<0.2 0.2 0.2>
    componentLinearVelScale=<0.2 0.2 0.2>
    gravityOverride {z=-500}
    physicsBodyDefinitions [{
      collisionType=CustomSphere
      sphereCollisionRadius=5
      boundBone {boneName=dyn_braid_r_03}
      boxExtents=<10 10 10>
      constraintSetup {angularLimitsMax=<0 7 0>, angularLimitsMin=<0 -7 -10>, angularTarget=<0 0 0>}
      localJointOffset=<2 0 0>
     }]
    sphericalLimits [{limitRadius=12, drivingBone {boneName=spine_05}, sphereLocalOffset=<-10 5 0>}]
   }
   Anim Dynamics - Bone: dyn_braid_r_04(ComponentPose=@n16, Alpha=1, ExternalForce="0,0,0") {
    bUseGravityOverride=true
    bUseSphericalLimits=true
    alphaBoolBlend {blendOption=Linear}
    boundBone {boneName=dyn_braid_r_04}
    componentAppliedLinearAccClamp=<3000 3000 3000>
    componentLinearAccScale=<0.2 0.2 0.2>
    componentLinearVelScale=<0.2 0.2 0.2>
    gravityOverride {z=-500}
    physicsBodyDefinitions [{
      collisionType=CustomSphere
      sphereCollisionRadius=5
      boundBone {boneName=dyn_braid_r_04}
      boxExtents=<10 10 10>
      constraintSetup {angularLimitsMax=<0 7 0>, angularLimitsMin=<0 -7 -15>, angularTarget=<0 0 0>}
      localJointOffset=<2 0 0>
     }]
    sphericalLimits [{limitRadius=12, drivingBone {boneName=spine_05}, sphereLocalOffset=<-10 9 0>}]
   }
   "Constraint - dyn_mask_off_mstr (P)"(Alpha=Get mask turn off(), ConstraintWeights_0=1) {
    alphaBoolBlend {blendOption=Linear}
    boneToModify {boneName=dyn_mask_off_mstr}
    constraintSetup [{targetBone {boneName=neck_01}}]
    constraintWeights [1]
   }
   "Constraint - dyn_braid_l_MSTR (P)"(Alpha=1, ConstraintWeights_0=1) {
    alphaBoolBlend {blendOption=Linear}
    boneToModify {boneName=dyn_braid_l_MSTR}
    constraintSetup [{targetBone {boneName=spine_05}}]
    constraintWeights [1]
   }
   n17 = Get mask turn off()
   n18 = "Constraint - dyn_braid_l_01 (PPPP)"(Alpha=@n17, ConstraintWeights_0=0.176, ConstraintWeights_1=0.2007, ConstraintWeights_2=0.1946, ConstraintWeights_3=0.4287) {
    alphaBoolBlend {blendOption=Linear}
    boneToModify {boneName=dyn_braid_l_01}
    constraintSetup [{targetBone {boneName=spine_05}}, {targetBone {boneName=neck_01}}, {targetBone {boneName=head}}, {targetBone {boneName=neck_02}}]
    constraintWeights [1,1,1,1]
   }
   n19 = "Constraint - dyn_braid_l_02 (PPP)"(ComponentPose=@n18, Alpha=@n17, ConstraintWeights_0=0.718, ConstraintWeights_1=0.2777, ConstraintWeights_2=0.0043) {
    alphaBoolBlend {blendOption=Linear}
    boneToModify {boneName=dyn_braid_l_02}
    constraintSetup [{targetBone {boneName=spine_05}}, {targetBone {boneName=clavicle_l}}, {targetBone {boneName=neck_02}}]
    constraintWeights [1,1,1]
   }
   n20 = "Constraint - dyn_braid_l_03 (PPP)"(ComponentPose=@n19, Alpha=@n17, ConstraintWeights_0=0.0207, ConstraintWeights_1=0.7578, ConstraintWeights_2=0.2215) {
    alphaBoolBlend {blendOption=Linear}
    boneToModify {boneName=dyn_braid_l_03}
    constraintSetup [{targetBone {boneName=spine_04}}, {targetBone {boneName=spine_05}}, {targetBone {boneName=clavicle_l}}]
    constraintWeights [1,1,1]
   }
   n21 = "Constraint - dyn_braid_r_01 (PPPP)"(
    ComponentPose=@n20
    Alpha=@n17
    ConstraintWeights_0=0.1752
    ConstraintWeights_1=0.2008
    ConstraintWeights_2=0.1961
    ConstraintWeights_3=0.4279
   ) {
    alphaBoolBlend {blendOption=Linear}
    boneToModify {boneName=dyn_braid_r_01}
    constraintSetup [{targetBone {boneName=spine_05}}, {targetBone {boneName=neck_01}}, {targetBone {boneName=head}}, {targetBone {boneName=neck_02}}]
    constraintWeights [1,1,1,1]
   }
   n22 = "Constraint - dyn_braid_r_02 (PPPP)"(ComponentPose=@n21, Alpha=@n17, ConstraintWeights_0=0.003, ConstraintWeights_1=0.7357, ConstraintWeights_2=0.257, ConstraintWeights_3=0.0043) {
    alphaBoolBlend {blendOption=Linear}
    boneToModify {boneName=dyn_braid_r_02}
    constraintSetup [{targetBone {boneName=spine_04}}, {targetBone {boneName=spine_05}}, {targetBone {boneName=clavicle_r}}, {targetBone {boneName=neck_02}}]
    constraintWeights [1,1,1,1]
   }
   n23 = "Constraint - dyn_braid_r_03 (PPP)"(ComponentPose=@n22, Alpha=@n17, ConstraintWeights_0=0.0145, ConstraintWeights_1=0.7615, ConstraintWeights_2=0.224) {
    alphaBoolBlend {blendOption=Linear}
    boneToModify {boneName=dyn_braid_r_03}
    constraintSetup [{targetBone {boneName=spine_04}}, {targetBone {boneName=spine_05}}, {targetBone {boneName=clavicle_r}}]
    constraintWeights [1,1,1]
   }
   n24 = Get mask turn off()
   n25 = Look At - Bone: dyn_braid_l_04(ComponentPose=@n23, Alpha=@n24) {
    alphaBoolBlend {blendOption=Linear}
    boneToModify {boneName=dyn_braid_l_04}
    lookAtLocation=<0 12 -8>
    lookAtTarget {boneReference {boneName=spine_05}}
    lookAt_Axis {axis=<1 0>}
   }
   Look At - Bone: dyn_braid_r_04(ComponentPose=@n25, Alpha=@n24) {
    alphaBoolBlend {blendOption=Linear}
    boneToModify {boneName=dyn_braid_r_04}
    lookAtLocation=<0 12 7>
    lookAtTarget {boneReference {boneName=spine_05}}
    lookAt_Axis {axis=<1 0>}
   }
   n26 = "Constraint - dyn_braid_l_01 (PPPP)"(Alpha=1, ConstraintWeights_0=0.176, ConstraintWeights_1=0.2007, ConstraintWeights_2=0.1946, ConstraintWeights_3=0.4287) {
    alphaBoolBlend {blendOption=Linear}
    alphaScaleBiasClamp {bInterpResult=true}
    boneToModify {boneName=dyn_braid_l_01}
    constraintSetup [{targetBone {boneName=spine_05}}, {targetBone {boneName=neck_01}}, {targetBone {boneName=head}}, {targetBone {boneName=neck_02}}]
    constraintWeights [1,1,1,1]
   }
   n27 = "Constraint - dyn_braid_l_02 (PPP)"(ComponentPose=@n26, Alpha=1, ConstraintWeights_0=0.718, ConstraintWeights_1=0.2777, ConstraintWeights_2=0.0043) {
    alphaBoolBlend {blendOption=Linear}
    alphaScaleBiasClamp {bInterpResult=true}
    boneToModify {boneName=dyn_braid_l_02}
    constraintSetup [{targetBone {boneName=spine_05}}, {targetBone {boneName=clavicle_l}}, {targetBone {boneName=neck_02}}]
    constraintWeights [1,1,1]
   }
   n28 = "Constraint - dyn_braid_l_03 (PPP)"(ComponentPose=@n27, Alpha=1, ConstraintWeights_0=0.0207, ConstraintWeights_1=0.7578, ConstraintWeights_2=0.2215) {
    alphaBoolBlend {blendOption=Linear}
    alphaScaleBiasClamp {bInterpResult=true}
    boneToModify {boneName=dyn_braid_l_03}
    constraintSetup [{targetBone {boneName=spine_04}}, {targetBone {boneName=spine_05}}, {targetBone {boneName=clavicle_l}}]
    constraintWeights [1,1,1]
   }
   n29 = "Constraint - dyn_braid_r_01 (PPPP)"(ComponentPose=@n28, Alpha=1, ConstraintWeights_0=0.1752, ConstraintWeights_1=0.2008, ConstraintWeights_2=0.1961, ConstraintWeights_3=0.4279) {
    alphaBoolBlend {blendOption=Linear}
    alphaScaleBiasClamp {bInterpResult=true}
    boneToModify {boneName=dyn_braid_r_01}
    constraintSetup [{targetBone {boneName=spine_05}}, {targetBone {boneName=neck_01}}, {targetBone {boneName=head}}, {targetBone {boneName=neck_02}}]
    constraintWeights [1,1,1,1]
   }
   n30 = "Constraint - dyn_braid_r_02 (PPPP)"(ComponentPose=@n29, Alpha=1, ConstraintWeights_0=0.003, ConstraintWeights_1=0.7357, ConstraintWeights_2=0.257, ConstraintWeights_3=0.0043) {
    alphaBoolBlend {blendOption=Linear}
    alphaScaleBiasClamp {bInterpResult=true}
    boneToModify {boneName=dyn_braid_r_02}
    constraintSetup [{targetBone {boneName=spine_04}}, {targetBone {boneName=spine_05}}, {targetBone {boneName=clavicle_r}}, {targetBone {boneName=neck_02}}]
    constraintWeights [1,1,1,1]
   }
   n31 = "Constraint - dyn_braid_r_03 (PPP)"(ComponentPose=@n30, Alpha=1, ConstraintWeights_0=0.0145, ConstraintWeights_1=0.7615, ConstraintWeights_2=0.224) {
    alphaBoolBlend {blendOption=Linear}
    alphaScaleBiasClamp {bInterpResult=true}
    boneToModify {boneName=dyn_braid_r_03}
    constraintSetup [{targetBone {boneName=spine_04}}, {targetBone {boneName=spine_05}}, {targetBone {boneName=clavicle_r}}]
    constraintWeights [1,1,1]
   }
   n32 = Look At - Bone: dyn_braid_l_04(ComponentPose=@n31, Alpha=1) {
    alphaBoolBlend {blendOption=Linear}
    alphaScaleBiasClamp {bInterpResult=true}
    boneToModify {boneName=dyn_braid_l_04}
    lookAtLocation=<0 12 -8>
    lookAtTarget {boneReference {boneName=spine_05}}
    lookAt_Axis {axis=<1 0>}
   }
   n33 = Look At - Bone: dyn_braid_r_04(ComponentPose=@n32, Alpha=1) {
    alphaBoolBlend {blendOption=Linear}
    alphaScaleBiasClamp {bInterpResult=true}
    boneToModify {boneName=dyn_braid_r_04}
    lookAtLocation=<1 12 7>
    lookAtTarget {boneReference {boneName=spine_05}}
    lookAt_Axis {axis=<1 0>}
   }
   "Transform (Modify) Bone - Bone: dyn_unmask_strap_l"(ComponentPose=@n33, Alpha=1, Translation="-1,0,0", Rotation="5,0,0", Scale="1,1,1") {
    rotationMode=BMM_Additive
    rotationSpace=BCS_BoneSpace
    translationMode=BMM_Additive
    translationSpace=BCS_BoneSpace
    alphaBoolBlend {blendOption=Linear}
    alphaScaleBiasClamp {bInterpResult=true, interpSpeedDecreasing=5, interpSpeedIncreasing=5}
    boneToModify {boneName=dyn_unmask_strap_l}
    rotation {pitch=5}
    translation {x=-1}
   }
   Output Pose(Result=Use cached pose 'fixes'()) {layerGroup=None}
   n34 = RigidBody!(ComponentPose=Local To Component(LocalPose=Use cached pose 'ctrlRig'()), Alpha) -> (Pose) {
    bOverrideWorldGravity=true
    overridePhysicsAsset=&/F_MED_GraffitiTonMask_FaceAcc_Physics_mask_up
    simulationSpace=BaseBoneSpace
    alphaBoolBlend {blendOption=Linear}
    baseBoneRef {boneName=spine_05}
    componentAppliedLinearAccClamp=<3000 3000 3000>
    overrideWorldGravity {z=-980}
    simSpaceSettings {maxAngularAcceleration=4, maxAngularVelocity=6, maxLinearAcceleration=600, maxLinearVelocity=1600, worldAlpha=1, externalLinearDragV=<1 1>}
   }
   n35 = Component To Local(ComponentPose=@n34)
   n36 = RigidBody(ComponentPose=Local To Component(LocalPose=Use cached pose 'ctrlRig'())) {
    bOverrideWorldGravity=true
    overridePhysicsAsset=&/F_MED_GraffitiTonMask_FaceAcc_Physics_mask_down
    simulationSpace=BaseBoneSpace
    alphaBoolBlend {blendOption=Linear}
    baseBoneRef {boneName=spine_05}
    componentAppliedLinearAccClamp=<3000 3000 3000>
    overrideWorldGravity {z=-980}
    simSpaceSettings {maxAngularAcceleration=4, maxAngularVelocity=6, maxLinearAcceleration=600, maxLinearVelocity=1600, worldAlpha=1, externalLinearDragV=<1 1>}
   }
   n37 = Component To Local(ComponentPose=@n36)
   n38 = Get isMaskOff() -> (isMaskOff:bool)
   n39 = Anim Dynamics - Bone: dyn_hair_small_uni(
    ComponentPose=Local To Component(LocalPose=Blend Poses by bool(BlendPose_0=@n37, BlendPose_1=@n35, bActiveValue=@n38) {blendType=Linear})
    Alpha=1
    ExternalForce="0,0,0"
   ) {
    angularDampingOverride=0.9
    angularSpringConstant=5000
    bAngularSpring=true
    bOverrideAngularDamping=true
    bUseGravityOverride=true
    alphaBoolBlend {blendOption=Linear}
    boundBone {boneName=dyn_hair_small_uni}
    componentAppliedLinearAccClamp=<800 800 800>
    componentLinearAccScale=<0.1 0.1 0.1>
    componentLinearVelScale=<0.2 0.2 0.2>
    gravityOverride {z=-500}
    physicsBodyDefinitions [{
      boundBone {boneName=dyn_hair_small_uni}
      boxExtents=<10 10 10>
      constraintSetup {angularLimitsMax=<0 0 15>, angularLimitsMin=<0 -10 -5>, angularTarget=<0 0 0>}
      localJointOffset=<0.7 0 0>
     }]
   }
   n40 = Anim Dynamics - Bone: dyn_hair_small_unmasked(ComponentPose=@n39, Alpha=1, ExternalForce="0,0,0") {
    angularDampingOverride=0.9
    angularSpringConstant=5000
    bAngularSpring=true
    bOverrideAngularDamping=true
    bUseGravityOverride=true
    alphaBoolBlend {blendOption=Linear}
    boundBone {boneName=dyn_hair_small_unmasked}
    componentAppliedLinearAccClamp=<800 800 800>
    componentLinearAccScale=<0.1 0.1 0.1>
    componentLinearVelScale=<0.2 0.2 0.2>
    gravityOverride {z=-650}
    physicsBodyDefinitions [{
      boundBone {boneName=dyn_hair_small_unmasked}
      boxExtents=<12 12 12>
      constraintSetup {angularLimitsMax=<0 0 5>, angularLimitsMin=<0 -10 -7>, angularTarget=<0 0 0>}
      localJointOffset=<1 0 0>
     }]
   }
   Save cached pose 'rigidBody'(Pose=Component To Local(ComponentPose=@n40)) {cacheName=rigidBody}
   n41 = Get arm_up_head_up() -> (arm_up_head_up:float)
   n42 = "Constraint - dyn_braid_l_02 (PPP)"(
    ComponentPose=Local To Component(LocalPose=Use cached pose 'rigidBody'())
    Alpha=@n41
    ConstraintWeights_0=0.718
    ConstraintWeights_1=0.2777
    ConstraintWeights_2=0.0043
   ) {
    alphaBoolBlend {blendOption=Linear}
    alphaScaleBiasClamp {bInterpResult=true}
    boneToModify {boneName=dyn_braid_l_02}
    constraintSetup [{targetBone {boneName=spine_05}}, {targetBone {boneName=clavicle_l}}, {targetBone {boneName=neck_02}}]
    constraintWeights [1,1,1]
   }
   n43 = "Constraint - dyn_braid_l_03 (PPP)"(ComponentPose=@n42, Alpha=@n41, ConstraintWeights_0=0.0207, ConstraintWeights_1=0.7578, ConstraintWeights_2=0.2215) {
    alphaBoolBlend {blendOption=Linear}
    boneToModify {boneName=dyn_braid_l_03}
    constraintSetup [{targetBone {boneName=spine_04}}, {targetBone {boneName=spine_05}}, {targetBone {boneName=clavicle_l}}]
    constraintWeights [1,1,1]
   }
   n44 = "Constraint - dyn_braid_l_04 (PPP)"!(ComponentPose=@n43, Alpha=@n41, ConstraintWeights_0=0.1621, ConstraintWeights_1:float=0.7752, ConstraintWeights_2:float=0.0627) -> (Pose) {
    alphaBoolBlend {blendOption=Linear}
    alphaScaleBiasClamp {bInterpResult=true}
    boneToModify {boneName=dyn_braid_l_04}
    constraintSetup [{targetBone {boneName=spine_04}}, {targetBone {boneName=spine_05}}, {targetBone {boneName=clavicle_l}}]
    constraintWeights [1,1,1]
   }
   n45 = "Constraint - dyn_braid_l_05 (PPP)"!(ComponentPose=@n44, Alpha=@n41, ConstraintWeights_0=0.3587, ConstraintWeights_1:float=0.6226, ConstraintWeights_2:float=0.0187) -> (Pose) {
    alphaBoolBlend {blendOption=Linear}
    alphaScaleBiasClamp {bInterpResult=true}
    boneToModify {boneName=dyn_braid_l_05}
    constraintSetup [{targetBone {boneName=spine_04}}, {targetBone {boneName=spine_05}}, {targetBone {boneName=clavicle_l}}]
    constraintWeights [1,1,1]
   }
   n46 = Look At - Bone: dyn_braid_l_04(ComponentPose=@n45, Alpha=@n41) {
    alphaBoolBlend {blendOption=Linear}
    boneToModify {boneName=dyn_braid_l_04}
    lookAtLocation=<0 20 -7>
    lookAtTarget {boneReference {boneName=spine_04}}
    lookAt_Axis {axis=<1 0>}
    lookUp_Axis {axis {x=-1}}
   }
   n47 = Get bIsSkydiveFloating()
   n48 = "Constraint - dyn_mask_off_mstr (P)"!(ComponentPose=@n46, bAlphaBoolEnabled:bool=@n47, ConstraintWeights_0=1) -> (Pose) {
    alphaInputType=Bool
    alphaBoolBlend {blendInTime=10, blendOption=Linear, blendOutTime=10}
    boneToModify {boneName=dyn_mask_off_mstr}
    constraintSetup [{targetBone {boneName=neck_01}}]
    constraintWeights [1]
   }
   n49 = "Transform (Modify) Bone - Bone: dyn_mask_off_mstr"!(ComponentPose=@n48, bAlphaBoolEnabled:bool=@n47, Translation="0,-2,2", Rotation="0,0,0", Scale="0.9,0.9,0.9") -> (Pose) {
    alphaInputType=Bool
    rotationMode=BMM_Additive
    rotationSpace=BCS_BoneSpace
    scaleMode=BMM_Replace
    scaleSpace=BCS_BoneSpace
    translationMode=BMM_Additive
    translationSpace=BCS_BoneSpace
    alphaBoolBlend {blendInTime=10, blendOption=Linear, blendOutTime=10}
    boneToModify {boneName=dyn_mask_off_mstr}
    scale=<0.9 0.9 0.9>
    translation {y=-2, z=2}
   }
   n50 = "Transform (Modify) Bone - Bone: dyn_mask_off_l"!(ComponentPose=@n49, bAlphaBoolEnabled:bool=@n47, Translation="0,0.5,0", Rotation="0,-10,0", Scale="1,1,1") -> (Pose) {
    alphaInputType=Bool
    rotationMode=BMM_Additive
    rotationSpace=BCS_BoneSpace
    translationMode=BMM_Additive
    translationSpace=BCS_BoneSpace
    alphaBoolBlend {blendInTime=10, blendOption=Linear, blendOutTime=10}
    boneToModify {boneName=dyn_mask_off_l}
    rotation {yaw=-10}
    translation {y=0.5}
   }
   n51 = "Transform (Modify) Bone - Bone: dyn_mask_off_r"!(ComponentPose=@n50, bAlphaBoolEnabled:bool=@n47, Translation="0,0.5,0", Rotation="0,10,0", Scale="1,1,1") -> (Pose) {
    alphaInputType=Bool
    rotationMode=BMM_Additive
    rotationSpace=BCS_BoneSpace
    translationMode=BMM_Additive
    translationSpace=BCS_BoneSpace
    alphaBoolBlend {blendInTime=10, blendOption=Linear, blendOutTime=10}
    boneToModify {boneName=dyn_mask_off_r}
    rotation {yaw=10}
    translation {y=0.5}
   }
   n52 = Get mask_turn_off2()
   n53 = "Transform (Modify) Bone - Bone: dyn_mask_off_mstr"!(ComponentPose=@n51, Alpha:float=@n52, Translation="-0.5,1,0", Rotation="0,0,0", Scale="0.9,0.9,0.9") -> (Pose) {
    rotationMode=BMM_Additive
    rotationSpace=BCS_BoneSpace
    scaleSpace=BCS_BoneSpace
    translationMode=BMM_Additive
    translationSpace=BCS_BoneSpace
    alphaBoolBlend {blendInTime=10, blendOption=Linear, blendOutTime=10}
    boneToModify {boneName=dyn_mask_off_mstr}
    scale=<0.9 0.9 0.9>
    translation=<-0.5 1>
   }
   n54 = "Transform (Modify) Bone - Bone: dyn_mask_off_l"!(ComponentPose=@n53, Alpha:float=@n52, Translation="0,0,0", Rotation="0,-10,0", Scale="1,1,1") -> (Pose) {
    rotationMode=BMM_Additive
    rotationSpace=BCS_BoneSpace
    scaleSpace=BCS_BoneSpace
    translationMode=BMM_Additive
    translationSpace=BCS_BoneSpace
    alphaBoolBlend {blendInTime=10, blendOption=Linear, blendOutTime=10}
    boneToModify {boneName=dyn_mask_off_l}
    rotation {yaw=-10}
   }
   n55 = "Transform (Modify) Bone - Bone: dyn_mask_off_r"!(ComponentPose=@n54, Alpha:float=@n52, Translation="0,0,0", Rotation="0,10,0", Scale="1,1,1") -> (Pose) {
    rotationMode=BMM_Additive
    rotationSpace=BCS_BoneSpace
    scaleSpace=BCS_BoneSpace
    translationMode=BMM_Additive
    translationSpace=BCS_BoneSpace
    alphaBoolBlend {blendInTime=10, blendOption=Linear, blendOutTime=10}
    boneToModify {boneName=dyn_mask_off_r}
    rotation {yaw=10}
   }
   Save cached pose 'fixes'(Pose=Component To Local(ComponentPose=@n55)) {cacheName=fixes}
   n56 = Control Rig!(Source=Copy Pose From Mesh() {bUseAttachedParent=true}, Alpha, mask_turn_off:double=Get mask turn off()) -> (Pose) {
    alphaBoolBlend {blendOption=Linear}
    controlRigAssetReference {blueprintRigClass=&/F_MED_GraffitiTon_FaceAcc_CtrlRig.F_MED_GraffitiTon_FaceAcc_CtrlRig_C}
   }
   Save cached pose 'ctrlRig'(Pose=@n56) {cacheName=ctrlRig}
   Get mask_turn_off2()
   Get mask turn off()
  ]}
 EventGraph {nodes [
   n1 = Event BlueprintUpdateAnimation()
   > Sequence()
   n2 = GetOwnerFortPlayerAnimInstance!(self:Custom Character Part Anim Instance Object Reference) -> (ReturnValue:Fort Player Anim Instance Object Reference)
   n3 = "To Float (Boolean)"(InBool=OR Boolean(
     A=Montage_IsPlaying(self=@n2, Montage=/Game/Animation/Game/MainPlayer/Montages/Melee/PlayerMeleeCombo)
     B=Montage_IsPlaying(self=@n2, Montage=/Game/Animation/Game/MainPlayer/Combat/Melee/DualWield/DualWield_Harvesting_Combo_M)
    ))
   > Set mask turn off(mask turn off:double=FInterpTo(Current=Get mask turn off(), Target=@n3, DeltaTime=@n1.DeltaTimeX, InterpSpeed=6)) -> (Output_Get:double)
   > Set mask_turn_off2(mask_turn_off2:double=FInterpTo(Current=0, Target=@n3, DeltaTime=@n1.DeltaTimeX, InterpSpeed=3)) -> (Output_Get:double)
   > Set arm_up_head_up(arm_up_head_up:float=float - float(A=GetCurveValue(CurveName=arm_up_head_up), B=Get mask turn off())) -> (Output_Get:float)
   > Set crazy head rotation(crazy head rotation:float=float - float(A=GetCurveValue(CurveName=crazy_head_rotation), B=Get mask turn off())) -> (Output_Get:float)
   Event OnCharacterCustomizationLoaded()
   > n4 = Branch(Condition=CharacterItemDefinitionHasAnyMetaTag(Tags=Make Array("[0]"=(TagName="Cosmetics.Variant.Property.Stage2"))))
   > Set isMaskOff(isMaskOff:bool=true) -> (Output_Get:bool)
   Set isMaskOff(execute=@n4.else, isMaskOff=false)
  ]}
]
variables [
 arm_up_head_up {type=number}
 crazy head rotation {type=number}
 isMaskOff {type=boolean}
 mask turn off {type=number}
 mask_turn_off2 {type=number}
]
